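; BP_BasilOkra_PostProcess is a Blueprint that removes a referenced post-process actor on lower-end platforms.
; On BeginPlay, where cosmetic events can execute, it checks that its PostProcessTarget is valid and destroys it on Nintendo Switch (including preview) or mobile.
; Other platforms leave the target untouched, and the construction script is empty.
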(event EventBeginPlay
  (bind _postprocesstarget (Variables|Default|GetPostProcessTarget))
  (Networking|CanExecuteCosmeticEvents
    (:True
      (Utilities|IsValid _postprocesstarget
        (:Is%20Valid
          (if (or (TouchPlatform|IsNintendoSwitchorNintendoSwitchPreview) (TouchPlatform|IsMobileGame))
            (Actor|DestroyActor _postprocesstarget)))))))

(fn ConstructionScript)
