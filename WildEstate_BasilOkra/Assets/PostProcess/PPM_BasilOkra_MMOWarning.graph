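# Post-process material that tints the screen edges with a pulsing orange warning vignette, apparently for an MMO warning state.
# A powered, smoothstepped distance from screen center masks the tint, and a time-driven sine oscillates its opacity between minimum and maximum parameters.
# That mask, scaled by BlendWeight and MMOPostProcessAlpha from the GlowDonut state parameter collection, blends the scene color toward orange.
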Material /WildEstate_BasilOkra/Assets/PostProcess/PPM_BasilOkra_MMOWarning
graph {
 properties {bAutomaticallySetUsageInEditor=false, materialDomain=MD_PostProcess, usedShadingModels=Default Lit}
 nodes [
  n1 = Distance(A=ScreenPosition().ViewportUV, B=Constant() {r=0.5})
  n2 = Sine(None=Multiply(A=Time(), B=ScalarParameter() {defaultValue=0.5, parameterName=PulseRate}))
  n3 = Multiply(
   A=SmoothStep(Value=Power(Base=@n1) {constExponent=4}) {constMax=0.24}
   B=LinearInterpolate(
    A=ScalarParameter() {defaultValue=0.2475, parameterName=OpacityMin}
    B=ScalarParameter() {defaultValue=0.4125, parameterName=OpacityMax}
    Alpha=ConstantBiasScale(None=@n2)
   )
  )
  n4 = ScalarParameter() {parameterName=BlendWeight}
  LinearInterpolate(
   A=ComponentMask(None=SceneTexture() {sceneTextureId=PPI_PostProcessInput0}.Color) {b=true, g=true, r=true}
   B=Constant3Vector() {constant=<rgba 1 0.3433 0 1>}
   Alpha=Multiply(
    A=Multiply(A=@n3, B=@n4)
    B=CollectionParameter() {collection=/WildEstateTech/TechArt/GlowDonut/Timer/MPC_GlowDonutState, parameterName=MMOPostProcessAlpha}
   )
  )
 ]
}
